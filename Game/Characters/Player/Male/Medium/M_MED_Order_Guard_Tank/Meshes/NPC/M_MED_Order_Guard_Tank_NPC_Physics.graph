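# Physics asset for the NPC mesh of the male medium Order Guard Tank character, covering its dangling belt gear.
# Capsule bodies sit on the pelvis, a back knife and three mirrored pairs of bag bones.
# Each gear body is constrained directly to the pelvis with small swing limits, tighter for the knife, and near-zero twist, so accessories can only sway slightly.
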PhysicsAsset /Game/Characters/Player/Male/Medium/M_MED_Order_Guard_Tank/Meshes/NPC/M_MED_Order_Guard_Tank_NPC_Physics
bodies {
 dyn_back_knife [{
   shapeType=Capsule
   length=15.4473
   radius=4.4712
   shapeName=dyn_back_knife_capsule
   center=<-5.3093 -0.1211 0.6231>
   rotation=<rot 1.0118 91.7921 -89.9999>
  }]
 dyn_bag_a_l [{
   shapeType=Capsule
   length=7.2992
   radius=4.0675
   shapeName=dyn_bag_a_l_capsule
   center=<-4.9791 -0.2487 2.3098>
   rotation=<rot 0 -86.4847 -81.695>
  }]
 dyn_bag_a_r [{
   shapeType=Capsule
   length=7.2992
   radius=4.0675
   shapeName=dyn_bag_a_r_capsule
   center=<4.9792 -0.2487 2.3098>
   rotation=<rot 0 86.4847 -81.695>
  }]
 dyn_bag_b_l [{
   shapeType=Capsule
   length=7.299
   radius=4.0492
   shapeName=dyn_bag_b_l_capsule
   center=<-5.1345 0.3958 2.8203>
   rotation=<rot 0 -85.9886 -86.0226>
  }]
 dyn_bag_b_r [{
   shapeType=Capsule
   length=7.299
   radius=4.0492
   shapeName=dyn_bag_b_r_capsule
   center=<5.1344 0.3958 2.8202>
   rotation=<rot 0 85.9886 -86.0226>
  }]
 dyn_bag_c_l [{
   shapeType=Capsule
   length=7.299
   radius=4.05
   shapeName=dyn_bag_c_l_capsule
   center=<-5.6056 0.9835 0.6129>
   rotation=<rot 0 -85.4747 -88.2526>
  }]
 dyn_bag_c_r [{shapeType=Capsule, length=7.299, radius=4.05, shapeName=dyn_bag_c_r_capsule, center=<5.6051 0.9835 0.613>, rotation=<rot 0 85.4746 -88.2526>}]
 pelvis [{shapeType=Capsule, length=18.6125, radius=7.9698, shapeName=pelvis_capsule, center=<7.3777 -0.8 0>, rotation=<rot 0.0001 -89.9999 -0.0001>}]
}
constraints [
 {bone1Name=dyn_bag_a_l, bone2Name=pelvis, swing1LimitDegrees=8, swing2LimitDegrees=8, twistLimitDegrees=0.01}
 {bone1Name=dyn_bag_b_l, bone2Name=pelvis, swing1LimitDegrees=8, swing2LimitDegrees=8, twistLimitDegrees=0.01}
 {bone1Name=dyn_bag_c_l, bone2Name=pelvis, swing1LimitDegrees=8, swing2LimitDegrees=8, twistLimitDegrees=0.01}
 {bone1Name=dyn_bag_a_r, bone2Name=pelvis, swing1LimitDegrees=8, swing2LimitDegrees=8, twistLimitDegrees=0.01}
 {bone1Name=dyn_bag_b_r, bone2Name=pelvis, swing1LimitDegrees=8, swing2LimitDegrees=8, twistLimitDegrees=0.01}
 {bone1Name=dyn_bag_c_r, bone2Name=pelvis, swing1LimitDegrees=8, swing2LimitDegrees=8, twistLimitDegrees=0.01}
 {bone1Name=dyn_back_knife, bone2Name=pelvis, swing1LimitDegrees=3, swing2LimitDegrees=3, twistLimitDegrees=0.01}
]
